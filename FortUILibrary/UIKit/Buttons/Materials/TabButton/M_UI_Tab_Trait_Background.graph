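# UI material for the translucent background of a trait tab button, varying by button state.
# The Slate post buffer feeds three CoreUI backplates as blur: C for idle/hovered, A for focused, pressed and selected, Disabled for disabled states.
# Modular-button lerp functions blend these, with static switch Is Selectable adding selected states.
# A rounded-corner mask multiplies opacity; the selectable lerp's Base States Opacity input takes the base Color output.
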Material /FortUILibrary/UIKit/Buttons/Materials/TabButton/M_UI_Tab_Trait_Background
graph {
 properties {
  bUsedWithStaticMesh=false
  blendMode=BLEND_Translucent
  materialDomain=MD_UI
  pixelDepthOffsetMode=PDOM_Legacy
  usedShadingModels=Default Lit
 }
 nodes [
  n1 = MaterialFunctionCall!("UVs (Optional)") {materialFunction=/Engine/Functions/UserInterface/GetSlatePost0}
  n2 = MaterialFunctionCall!(Post Buffer Blur=@n1.LinearRGB) {materialFunction=/FortUILibrary/MaterialFunctions/CoreUI/MF_UI_CoreUI_BackplateC}
  n3 = MaterialFunctionCall!(Post Buffer Blur=@n1.LinearRGB) {materialFunction=/FortUILibrary/MaterialFunctions/CoreUI/MF_UI_CoreUI_BackplateA}
  n4 = MaterialFunctionCall!(Post Buffer Blur=@n1.LinearRGB) {materialFunction=/FortUILibrary/MaterialFunctions/CoreUI/MF_UI_CoreUI_BackplateDisabled}
  n5 = MaterialFunctionCall!(
   Idle=@n2.Result
   Idle Opacity
   Hovered=@n2.Result
   Hovered Opacity
   Focused=@n3.Result
   Focused Opacity
   Pressed=@n3.Result
   Pressed Opacity
   Disabled=@n4.Result
   Disabled Opacity
  ) {materialFunction=/FortUILibrary/Systems/ModularButton/MaterialFunctions/MF_UIKit_ButtonStatesBase_Lerp}
  n6 = MaterialFunctionCall!(
   Base States=@n5.Color
   Base States Opacity=@n5.Color
   Selected_Idle=@n3.Result
   Selected_Idle Opacity
   Selected_Hovered=@n3.Result
   Selected_Hovered Opacity
   Selected_Focused=@n3.Result
   Selected_Focused Opacity
   Selected_Pressed=@n3.Result
   Selected_Pressed Opacity
   Selected_Disabled=@n4.Result
   Selected_Disabled Opacity
  ) {materialFunction=/FortUILibrary/Systems/ModularButton/MaterialFunctions/MF_UIKit_ButtonStatesSelectable_Lerp}
  Multiply(A=StaticSwitchParameter(True=@n6.Opacity, False=@n5.Opacity) {parameterName=Is Selectable}, B=NamedRerouteUsage())
  NamedRerouteDeclaration(None=MaterialFunctionCall(
    UV=NamedRerouteUsage()
    Corner Radius=ScalarParameter() {defaultValue=10, parameterName=Corner Radius}
    Radius in Pixels=ScalarParameter() {defaultValue=1, parameterName=Radius in Pixels, sliderMax=1}
   ) {materialFunction=&MF_UI_CoreUI_RoundedMask}.Shape Mask) {name=Fill_Box, nodeColor=<rgba 0 0 0 1>}
  NamedRerouteDeclaration(None=MaterialFunctionCall() {materialFunction=&GetUserInterfaceUV}.Normalized UV) {name=Normalized UV, nodeColor=<rgba 0 0 0 1>}
  StaticSwitchParameter(True=@n6.Color, False=@n5.Color) {parameterName=Is Selectable}
 ]
}
